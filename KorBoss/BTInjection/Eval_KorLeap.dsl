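; Behavior-injection evaluator for the Kor boss that chooses leap destinations and gates when the leap behavior runs.
; Leap queries, triggered on enter, cooldown or Pistol/DMR behavior status changes, may randomly switch aggro target.
; EQS picks far leaps until a hotfix-tunable random count, then a short leap; failures fall back from far to short to spawn location.
; Results are written to the pawn and request re-evaluation, which requires CanLeap and being inside the leash.
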(event EventOnEnter
  (params (OwnerComp :type Behavior%20Tree%20Component%20Object%20Reference))
  (bind _asfort_athena_aibot_controller (Utilities|Casting|CastToFortAthenaAIBotController (Components|GetOwner OwnerComp)))
  (bind _output_get (Variables|Default|SetKorController _asfort_athena_aibot_controller))
  (AlertLevel|BindEventtoOnAlertLevelChangedEvent _output_get (EventDispatchers|CreateEvent :!event Custom|AlertLevelChangedEvent))
  (bind _asbp_player_pawn_cherry_smoke_kor_boss (Utilities|Casting|CastToBP_PlayerPawn_CherrySmoke_KorBoss (Pawn|GetControlledPawn _output_get)))
  (Variables|Default|SetKorPawn _asbp_player_pawn_cherry_smoke_kor_boss)
  (CallFunction|ExecuteLeapQuery))

(event Custom|TryShortLeap
  (bind _returnvalue (AI|EQS|RunEQSQuery "/KorBoss/EQS/EQS_FindKorShortLeapPoints.EQS_FindKorShortLeapPoints" (Variables|Default|GetKorPawn) "RandomBest5Pct"))
  (AI|EQS|BindEventtoOnQueryFinished _returnvalue (delegate Custom|ShortLeapLocationQueryFinished)))

(event Custom|ShortLeapLocationQueryFinished
  (params (QueryInstance :type EQS%20Query%20Instance%20Object%20Reference) (QueryStatus :type EEnvQueryStatus%20Enum :default "Processing"))
  (bind _korpawn (Variables|Default|GetKorPawn))
  (bind (_resultlocations _returnvalue) (AI|EQS|GetQueryResultsasLocations QueryInstance))
  (if (and _returnvalue (Utilities|Operators|NotEqual(!=) :A (Utilities|Array|Get(acopy) _resultlocations)))
    (bind _output_get (Variables|/KorBoss/Pawn/BP_PlayerPawn_CherrySmoke_KorBoss.BP_PlayerPawn_CherrySmoke_KorBoss_C|SetNextLeapLocation (Utilities|Array|Get(acopy) _resultlocations) _korpawn))
    (Variables|/KorBoss/Pawn/BP_PlayerPawn_CherrySmoke_KorBoss.BP_PlayerPawn_CherrySmoke_KorBoss_C|SetLastLeapLocation _output_get _korpawn)
    (Variables|Default|SetCanLeap true)
    (Variables|/KorBoss/Pawn/BP_PlayerPawn_CherrySmoke_KorBoss.BP_PlayerPawn_CherrySmoke_KorBoss_C|SetDoShortLeap true (Variables|Default|GetKorPawn))
    (Variables|Default|SetFarLeapAmount 0)
    (Variables|Default|SetLeapSwitchAmount 0)
    (CallFunction|RequestEvaluation)
    (else
      (CallFunction|LeapFallback))))

(event Custom|ExecuteLeapQuery
  (bind _korpawn (Variables|Default|GetKorPawn))
  (bind _korpawn_2 (Variables|Default|GetKorPawn))
  (Utilities|Time|ClearandInvalidateTimerbyHandle (Variables|Default|GetLeap%20Cooldown%20Handle))
  (Variables|/KorBoss/Pawn/BP_PlayerPawn_CherrySmoke_KorBoss.BP_PlayerPawn_CherrySmoke_KorBoss_C|SetNextLeapLocation "0, 0, 0" _korpawn_2)
  (Variables|/KorBoss/Pawn/BP_PlayerPawn_CherrySmoke_KorBoss.BP_PlayerPawn_CherrySmoke_KorBoss_C|SetDoFarLeap false _korpawn_2)
  (Variables|/KorBoss/Pawn/BP_PlayerPawn_CherrySmoke_KorBoss.BP_PlayerPawn_CherrySmoke_KorBoss_C|SetDoShortLeap false _korpawn_2)
  (if (Math|Random|RandomBoolwithWeight (ScalableFloat|GetValueatLevel (Variables|Hotfix|GetTargetSwitchWeight_Hf)))
    (Class|BPPlayerPawnCherrySmokeKorBoss|TrySwitchAggroTarget _korpawn_2)
    (else
      _))
  (if (== (Variables|Default|GetLeapSwitchAmount) 0)
    (Variables|Default|SetLeapSwitchAmount (Math|Random|RandomIntegerinRange (ScalableFloat|GetValueasInteger (Variables|Hotfix|GetMinFarLeapAmount_Hf)) (ScalableFloat|GetValueasInteger (Variables|Hotfix|GetMaxFarLeapAmount_Hf))))
    (else
      _))
  (if (>= (Variables|Default|GetFarLeapAmount) (Variables|Default|GetLeapSwitchAmount))
    (bind _returnvalue (AI|EQS|RunEQSQuery "/KorBoss/EQS/EQS_FindKorShortLeapPoints.EQS_FindKorShortLeapPoints" _korpawn "RandomBest5Pct"))
    (AI|EQS|BindEventtoOnQueryFinished _returnvalue (delegate Custom|ShortLeapLocationQueryFinished))
    (else
      (bind _returnvalue_2 (AI|EQS|RunEQSQuery "/KorBoss/EQS/EQS_FindKorFarLeapPoints.EQS_FindKorFarLeapPoints" _korpawn "RandomBest5Pct"))
      (AI|EQS|BindEventtoOnQueryFinished _returnvalue_2 (delegate Custom|FarLeapLocationQueryFinished)))))

(event Custom|FarLeapLocationQueryFinished
  (params (QueryInstance :type EQS%20Query%20Instance%20Object%20Reference) (QueryStatus :type EEnvQueryStatus%20Enum :default "Processing"))
  (bind _korpawn (Variables|Default|GetKorPawn))
  (bind (_resultlocations _returnvalue) (AI|EQS|GetQueryResultsasLocations QueryInstance))
  (if (and _returnvalue (Utilities|Operators|NotEqual(!=) :A (Utilities|Array|Get(acopy) _resultlocations)))
    (bind _output_get (Variables|/KorBoss/Pawn/BP_PlayerPawn_CherrySmoke_KorBoss.BP_PlayerPawn_CherrySmoke_KorBoss_C|SetNextLeapLocation (Utilities|Array|Get(acopy) _resultlocations) _korpawn))
    (Variables|/KorBoss/Pawn/BP_PlayerPawn_CherrySmoke_KorBoss.BP_PlayerPawn_CherrySmoke_KorBoss_C|SetLastLeapLocation _output_get _korpawn)
    (Variables|Default|SetCanLeap true)
    (Variables|/KorBoss/Pawn/BP_PlayerPawn_CherrySmoke_KorBoss.BP_PlayerPawn_CherrySmoke_KorBoss_C|SetDoFarLeap true (Variables|Default|GetKorPawn))
    (Math|Integer|IncrementInt (Variables|Default|GetFarLeapAmount)
      (:%20%20
        (CallFunction|RequestEvaluation)))
    (else
      (CallFunction|TryShortLeap))))

(event Custom|AlertLevelChangedEvent
  (params (BotController :type Fort%20Athena%20AIBot%20Controller%20Object%20Reference) (OldAlertLevel :type EAlertLevel%20Enum :default "Unaware") (NewAlertLevel :type EAlertLevel%20Enum :default "Unaware"))
  (switch Utilities|FlowControl|Switch|SwitchonEAlertLevel NewAlertLevel
    (:Unaware
      (Variables|Default|SetFarLeapAmount 0))))

(event EventOnExit
  (params (OwnerComp :type Behavior%20Tree%20Component%20Object%20Reference))
  (Utilities|Time|ClearandInvalidateTimerbyHandle (Variables|Default|GetLeap%20Cooldown%20Handle))
  (Variables|Default|SetCanLeap false))

(event EventOnConcurentBehaviorExecutionStatusChanged
  (params (InjectedBehavior :type Fort%20Injected%20Behavior%20Object%20Reference) (PreviousExecutionStatus :type EExecutionStatus%20Enum :default "ExecutionError") (NewExecutionStatus :type EExecutionStatus%20Enum :default "ExecutionError"))
  (bind _returnvalue (Utilities|GetClass InjectedBehavior))
  (bind _pistol_ib (Variables|Default|GetPistol_IB))
  (bind _returnvalue_2 (== _returnvalue _pistol_ib))
  (bind _dmr_ib (Variables|Default|GetDMR_IB))
  (bind _returnvalue_3 (== _dmr_ib _returnvalue))
  (bind _returnvalue_4 (or _returnvalue_3 _returnvalue_2))
  (if (Math|Boolean|ANDBoolean _returnvalue_4 (or (Utilities|Enum|Equal(Enum) NewExecutionStatus "ExecutionDenied") (Utilities|Enum|Equal(Enum) NewExecutionStatus "ExecutionError")) (Variables|/KorBoss/Pawn/BP_PlayerPawn_CherrySmoke_KorBoss.BP_PlayerPawn_CherrySmoke_KorBoss_C|GetForceLeap (Variables|Default|GetKorPawn)))
    (CallFunction|ExecuteLeapQuery)
    (elif (and _returnvalue_4 (or (Utilities|Enum|Equal(Enum) NewExecutionStatus "ExecutionSuccess") (Utilities|Enum|Equal(Enum) NewExecutionStatus "ExecutionError")))
      (CallFunction|ExecuteLeapQuery))))

(event EventOnExecutionStatusChanged
  (params (PreviousExecutionStatus :type EExecutionStatus%20Enum :default "ExecutionError") (NewExecutionStatus :type EExecutionStatus%20Enum :default "ExecutionError"))
  (if (Utilities|Enum|Equal(Enum) NewExecutionStatus "ExecutionPending")
    (Variables|Default|SetCanLeap false)))

(fn ActivateLeapCooldown
  (bind _returnvalue (Utilities|Time|SetTimerbyEvent (EventDispatchers|CreateEvent :!event ExecuteLeapQuery) (ScalableFloat|GetValueatLevel (Variables|Hotfix|GetLeapCooldown_Hf))))
  (Variables|Default|SetLeap%20Cooldown%20Handle _returnvalue))

(fn EvaluateOngoingConditions
  (params (OwnerComp :type Behavior%20Tree%20Component%20Object%20Reference))
  (|CallParentFunction OwnerComp)
  (return true))

(fn EvaluateStartingConditions
  (params (OwnerComp :type Behavior%20Tree%20Component%20Object%20Reference))
  (|CallParentFunction OwnerComp)
  (return (and (Variables|Default|GetCanLeap) (Class|FortAthenaAIBotController|IsLocationInsideLeash (Variables|Default|GetKorController) (Transformation|GetActorLocation (Variables|Default|GetKorPawn)) false 0.0))))

(fn LeapFallback
  (bind _korpawn (Variables|Default|GetKorPawn))
  (bind _output_get (Variables|/KorBoss/Pawn/BP_PlayerPawn_CherrySmoke_KorBoss.BP_PlayerPawn_CherrySmoke_KorBoss_C|SetNextLeapLocation (Variables|/KorBoss/Pawn/BP_PlayerPawn_CherrySmoke_KorBoss.BP_PlayerPawn_CherrySmoke_KorBoss_C|GetSpawnLocation _korpawn) _korpawn))
  (Variables|/KorBoss/Pawn/BP_PlayerPawn_CherrySmoke_KorBoss.BP_PlayerPawn_CherrySmoke_KorBoss_C|SetLastLeapLocation _output_get _korpawn)
  (Variables|Default|SetCanLeap true)
  (Variables|/KorBoss/Pawn/BP_PlayerPawn_CherrySmoke_KorBoss.BP_PlayerPawn_CherrySmoke_KorBoss_C|SetDoShortLeap true (Variables|Default|GetKorPawn))
  (Variables|Default|SetFarLeapAmount 0)
  (Variables|Default|SetLeapSwitchAmount 0)
  (CallFunction|RequestEvaluation))
