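; Behavior tree injection evaluator for the Cherry Smoke Kor boss that, going by its name, gates a return-to-islands behavior.
; On enter it caches the owning Fort Athena AI bot controller and the controlled Kor boss pawn.
; Starting conditions pass only when the pawn is arena-bound and its location is outside the controller's leash.
; Ongoing conditions always return true, and both evaluations call the parent but ignore its result.
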(event EventOnEnter
  (params (OwnerComp :type Behavior%20Tree%20Component%20Object%20Reference))
  (bind _asfort_athena_aibot_controller (Utilities|Casting|CastToFortAthenaAIBotController (Components|GetOwner OwnerComp)))
  (bind _output_get (Variables|Default|SetKorController _asfort_athena_aibot_controller))
  (bind _asbp_player_pawn_cherry_smoke_kor_boss (Utilities|Casting|CastToBP_PlayerPawn_CherrySmoke_KorBoss (Pawn|GetControlledPawn _output_get)))
  (Variables|Default|SetKorPawn _asbp_player_pawn_cherry_smoke_kor_boss))

(fn EvaluateOngoingConditions
  (params (OwnerComp :type Behavior%20Tree%20Component%20Object%20Reference))
  (|CallParentFunction OwnerComp)
  (return true))

(fn EvaluateStartingConditions
  (params (OwnerComp :type Behavior%20Tree%20Component%20Object%20Reference))
  (|CallParentFunction OwnerComp)
  (if (Variables|/KorBoss/Pawn/BP_PlayerPawn_CherrySmoke_KorBoss.BP_PlayerPawn_CherrySmoke_KorBoss_C|GetIsArenaBound (Variables|Default|GetKorPawn))
    (return (not (Class|FortAthenaAIBotController|IsLocationInsideLeash (Variables|Default|GetKorController) (Transformation|GetActorLocation (Variables|Default|GetKorPawn)) false 0.0)))
    (else
      (return false))))
