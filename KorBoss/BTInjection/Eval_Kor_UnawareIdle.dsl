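; Behavior tree evaluator that decides when the Kor boss runs its unaware idle behavior.
; On enter it caches the Kor controller and pawn; the pawn's DoUnawareIdle flag alone keeps the behavior running.
; While the boss is alive, it starts if that flag is set, or sets it when the target is missing, downed, or outside the leash and unseen beyond a delay.
; A server-time stamp tracks that delay, resetting when those conditions break.
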(event EventOnEnter
  (params (OwnerComp :type Behavior%20Tree%20Component%20Object%20Reference))
  (bind _asfort_athena_aibot_controller (Utilities|Casting|CastToFortAthenaAIBotController (Components|GetOwner OwnerComp)))
  (bind _output_get (Variables|Default|SetKorController _asfort_athena_aibot_controller))
  (bind _asbp_player_pawn_cherry_smoke_kor_boss (Utilities|Casting|CastToBP_PlayerPawn_CherrySmoke_KorBoss (Pawn|GetControlledPawn _output_get)))
  (Variables|Default|SetKorPawn _asbp_player_pawn_cherry_smoke_kor_boss))

(fn EvaluateOngoingConditions
  (params (OwnerComp :type Behavior%20Tree%20Component%20Object%20Reference))
  (|CallParentFunction OwnerComp)
  (return (Variables|/KorBoss/Pawn/BP_PlayerPawn_CherrySmoke_KorBoss.BP_PlayerPawn_CherrySmoke_KorBoss_C|GetDoUnawareIdle (Variables|Default|GetKorPawn))))

(fn EvaluateStartingConditions
  (params (OwnerComp :type Behavior%20Tree%20Component%20Object%20Reference))
  (bind _korpawn (Variables|Default|GetKorPawn))
  (bind _returnvalue (FortPawn|IsDead _korpawn))
  (bind _returnvalue_2 (not _returnvalue))
  (bind _korcontroller (Variables|Default|GetKorController))
  (|CallParentFunction OwnerComp)
  (if (and (Variables|/KorBoss/Pawn/BP_PlayerPawn_CherrySmoke_KorBoss.BP_PlayerPawn_CherrySmoke_KorBoss_C|GetDoUnawareIdle (Variables|Default|GetKorPawn)) (not (FortPawn|IsDead (Variables|Default|GetKorPawn))))
    (return true)
    (else
      (bind _targetactor (Class|BPFLAICommonFunctions|GetTargetActorfromBlackboard _korpawn))
      (Utilities|IsValid _targetactor
        (:Is%20Valid
          (if (Health|IsDBNO _targetactor)
            (if (not _returnvalue_2) :!label _label_3
              (return false))
            (elif (not (Class|FortAthenaAIBotController|IsLocationInsideLeash _korcontroller (Transformation|GetActorLocation _targetactor) false 0.0))
              (if (== (Variables|Default|GetTargetLostDelay) 0)
                (Variables|Default|SetTargetLostDelay (Fortnite|Sync|Time|GetServerWorldRealTimeSeconds)))
              (if (and _returnvalue_2 (not (Controller|LineOfSightTo :self _korcontroller :Other _targetactor :bAlternateChecks false)))
                (if (> (Math|Float|Absolute(Float) (- (Variables|Default|GetTargetLostDelay) (Fortnite|Sync|Time|GetServerWorldRealTimeSeconds))) (Variables|Default|GetDelayTimeSinceLeashExit))
                  (Variables|/KorBoss/Pawn/BP_PlayerPawn_CherrySmoke_KorBoss.BP_PlayerPawn_CherrySmoke_KorBoss_C|SetDoUnawareIdle true (Variables|Default|GetKorPawn) :!label _label_2)
                  (return true)
                  (else
                    _))
                (else
                  (Variables|Default|SetTargetLostDelay 0.0 :!label _label)))
              (return false)
              (else
                (goto _label))))
          (goto _label_2))
        (:Is%20Not%20Valid
          _))
      (goto _label_3))))
