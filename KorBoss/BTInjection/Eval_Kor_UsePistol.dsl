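; Behavior-injection evaluator that governs when the Kor boss AI uses its pistol behavior.
; On enter it caches the Kor controller and pawn, and it starts only while the pawn's DoShortLeap flag is set.
; It continues until the damage threshold is reached or the pawn is falling.
; It clears DoShortLeap on exit, or when the concurrent pistol injected behavior's status changes to pending or error.
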(event EventOnEnter
  (params (OwnerComp :type Behavior%20Tree%20Component%20Object%20Reference))
  (bind _asfort_athena_aibot_controller (Utilities|Casting|CastToFortAthenaAIBotController (Components|GetOwner OwnerComp)))
  (bind _output_get (Variables|Default|SetKorController _asfort_athena_aibot_controller))
  (bind _asbp_player_pawn_cherry_smoke_kor_boss (Utilities|Casting|CastToBP_PlayerPawn_CherrySmoke_KorBoss (Pawn|GetControlledPawn _output_get)))
  (Variables|Default|SetKorPawn _asbp_player_pawn_cherry_smoke_kor_boss))

(event EventOnExit
  (params (OwnerComp :type Behavior%20Tree%20Component%20Object%20Reference))
  (Variables|/KorBoss/Pawn/BP_PlayerPawn_CherrySmoke_KorBoss.BP_PlayerPawn_CherrySmoke_KorBoss_C|SetDoShortLeap false (Variables|Default|GetKorPawn)))

(event EventOnConcurentBehaviorExecutionStatusChanged
  (params (InjectedBehavior :type Fort%20Injected%20Behavior%20Object%20Reference) (PreviousExecutionStatus :type EExecutionStatus%20Enum :default "ExecutionError") (NewExecutionStatus :type EExecutionStatus%20Enum :default "ExecutionError"))
  (if (and (== (Variables|Default|GetPistol_IB) (Utilities|GetClass InjectedBehavior)) (or (Utilities|Enum|Equal(Enum) NewExecutionStatus "ExecutionPending") (Utilities|Enum|Equal(Enum) NewExecutionStatus "ExecutionError")))
    (Variables|/KorBoss/Pawn/BP_PlayerPawn_CherrySmoke_KorBoss.BP_PlayerPawn_CherrySmoke_KorBoss_C|SetDoShortLeap false (Variables|Default|GetKorPawn))))

(fn EvaluateOngoingConditions
  (params (OwnerComp :type Behavior%20Tree%20Component%20Object%20Reference))
  (|CallParentFunction OwnerComp)
  (return (and (not (Variables|/KorBoss/Pawn/BP_PlayerPawn_CherrySmoke_KorBoss.BP_PlayerPawn_CherrySmoke_KorBoss_C|GetDmgThresholdReached (Variables|Default|GetKorPawn))) (not (GameplayTags|HasMatchingGameplayTag (Variables|Default|GetKorPawn) "(TagName=\"MovementMode.Falling\")")))))

(fn EvaluateStartingConditions
  (params (OwnerComp :type Behavior%20Tree%20Component%20Object%20Reference))
  (|CallParentFunction OwnerComp)
  (return (Variables|/KorBoss/Pawn/BP_PlayerPawn_CherrySmoke_KorBoss.BP_PlayerPawn_CherrySmoke_KorBoss_C|GetDoShortLeap (Variables|Default|GetKorPawn))))
